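; Gameplay ability for the Season 12 Meatball quest that credits boat air time to the player's quest manager.
; On activation it ends unless the pawn matches the InBoat tag check, otherwise it starts a looping one-second timer.
; Each tick increments a counter while the vehicle is airborne; once it is not, a count of at least two sends one AirTime stat event.
; The counter resets on grounded ticks, and ability end also clears the timer.
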(event Ability|EventActivateAbilityFromEvent
  (params (EventData :type Gameplay%20Event%20Data%20Structure%20%28by%20ref%29))
  (|CallParentFunction EventData)
  (if (GameplayTags|TagContainer|CompareTagAssetInterfacetoOtherTags (Variables|Default|GetT_InBoat) (Variables|GAVars|Data|GetOwningPlayerPawn))
    (bind _returnvalue (Utilities|Time|SetTimerbyEvent (delegate Custom|Check%20If%20in%20Air) 1.0 true))
    (Variables|Default|SetTimer%20Handle _returnvalue)
    (else
      (Ability|EndAbility))))

(event Custom|Check%20If%20in%20Air
  (if (Vehicle|VehicleIsinAir (Vehicle|GetVehicleInterface (Variables|GAVars|Data|GetOwningPlayerPawn)))
    (Math|Integer|IncrementInt (Variables|Default|GetCounter))
    (elif (>= (Variables|Default|GetCounter) 2)
      (Quest|SendComplexCustomStatEvent (Variables|GAVars|Data|GetOwningPlayerQuestManager) 0 "(GameplayTags=((TagName=\"Athena.Quests.Meatball.AirTime\")))" "(GameplayTags=)" 1)
      (Variables|Default|SetCounter 0)
      (else
        (Variables|Default|SetCounter 0)))))

(event Ability|EventOnEndAbility
  (params (bWasCancelled :type Boolean))
  (Utilities|Time|ClearandInvalidateTimerbyHandle (Variables|Default|GetTimer%20Handle))
  (Variables|Default|SetCounter 0))
